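# Translucent unlit Niagara sprite material for a chrome environment nano-mist effect.
# Four differently panned samples of one texture have their green channels remapped, multiplied in pairs, summed and scaled by particle alpha.
# This mist value drives emissive and opacity, with color from parameters and particle color, hue-shifted by the red channel, then desaturated.
# Perlin noise panned in radial coordinates varies a sphere mask's radius, breaking up the sprite's circular edge.
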Material /Game/Effects/Fort_Effects/Effects/Environment/Chrome/Materials/M_NanoMist_Env
graph {
 properties {bUsedWithNiagaraSprites=true, blendMode=BLEND_Translucent, pixelDepthOffsetMode=PDOM_Legacy, shadingModel=MSM_Unlit, usedShadingModels=Unlit}
 nodes [
  n1 = TextureObjectParameter() {
   group=Main
   parameterName=Main_Texture
   samplerType=SAMPLERTYPE_Normal
   sortPriority=0
   texture=&/T_Chrome_Sand_Pickup2
  }
  n2 = VectorParameter() {group=Main, parameterName=Main_UVControl, channelNames {a=VPan, b=UPan, g=VScale, r=UScale}, defaultValue=<rgba 0.5 0.5 0.06 0.035>}
  n3 = Add(A=Time(), B=Constant() {r=1})
  n4 = MaterialFunctionCall(Texture=@n1, UVControl_RGB=@n2, UVControl_A=@n2.VPan, "[Override]TimeInput"=@n3) {materialFunction=&FXCommon_TextureSample}
  n5 = VectorParameter() {parameterName=TriC, defaultValue=<rgba 0 0 0 1>}
  n6 = VectorParameter() {parameterName=Color, defaultValue=<rgba 1 0 0 1>}
  n7 = ParticleColor()
  n8 = MaterialFunctionCall(Hue Shift Percentage=@n4.R, Texture=Add(A=Multiply(B=@n5.RGB), B=Multiply(A=@n6.RGB, B=@n7.RGB))) {materialFunction=&HueShift}
  n9 = VectorParameter() {group=Sec, parameterName=2_UVControl, channelNames {a=VPan, b=UPan, g=VScale, r=UScale}, defaultValue=<rgba 0.5 0.5 0.06 -0.04>}
  n10 = Multiply(A=Add(A=MaterialFunctionCall(Texture=@n1, UVControl_RGB=@n9, UVControl_A=@n9, "[Override]TimeInput"=@n3) {materialFunction=&FXCommon_TextureSample}.G)) {constB=0.5}
  n11 = VectorParameter() {group=Sec, parameterName=3_UVControl, channelNames {a=VPan, b=UPan, g=VScale, r=UScale}, defaultValue=<rgba 0.45 0.45 -0.05 0.05>}
  n12 = Multiply(A=Add(A=MaterialFunctionCall(Texture=@n1, UVControl_RGB=@n11, UVControl_A=@n11, "[Override]TimeInput"=@n3) {materialFunction=&FXCommon_TextureSample}.G) {constB=0.5}) {constB=0.5}
  n13 = Multiply(A=Add(A=@n4.G)) {constB=0.5}
  n14 = VectorParameter() {group=Sec, parameterName=1_UVControl, channelNames {a=VPan, b=UPan, g=VScale, r=UScale}, defaultValue=<rgba 0.45 0.45 -0.05 -0.05>}
  n15 = Multiply(A=Add(A=MaterialFunctionCall(Texture=@n1, UVControl_RGB=@n14, UVControl_A=@n14.VPan, "[Override]TimeInput"=@n3) {materialFunction=&FXCommon_TextureSample}.G) {constB=0.5}) {constB=0.5}
  n16 = Multiply(A=Add(A=Multiply(A=@n13, B=@n15), B=Multiply(A=@n10, B=@n12)), B=@n7.A)
  MaterialFunctionCall(Emissive=Desaturation(None=Multiply(A=@n8.Result, B=@n16), Fraction=ScalarParameter() {parameterName=Desaturate})) {materialFunction=&MF_FX_EmissiveAdjustments}
  n17 = Constant() {r=0.35}
  n18 = MaterialFunctionCall(Float2=MaterialFunctionCall() {materialFunction=&VectorToRadialValue}.Radial Coordinates) {materialFunction=&BreakOutFloat2Components}
  n19 = Multiply(A=@n18.R)
  n20 = Multiply(A=@n18.G) {constB=0.25}
  n21 = TextureCoordinate()
  n22 = TextureSample!(
   UVs=Panner(Coordinate=AppendVector(A=@n19, B=@n20)) {speedX=0.05, speedY=-0.1}
   Tex
   "DDX(UVs)"=DDX(Value=@n21)
   "DDY(UVs)"=DDY(Value=@n21)
   Apply View MipBias
  ) {mipValueMode=TMVM_Derivative, texture=/Game/Global/Textures/Default/Noise/3xPerlin_M}
  n23 = Multiply(A=@n22.R) {constB=0.35}
  Multiply(A=Saturate(None=@n16), B=SphereMask(A=TextureCoordinate(), B=Constant() {r=0.5}, Radius=Add(A=@n17, B=@n23), Hardness=Constant()))
  Reroute(None=@n22.B)
 ]
}
